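# Physics asset for the hoop bones of the F_MED_HighTower_Mango_Dark female character body, defining a small jointed chain for physical motion.
# Three small capsule bodies sit on dyn_parent_hoop_01, dyn_hoop_01 and dyn_hoop_02, with the last capsule offset and rotated.
# Two constraints link them in sequence from the parent bone, with tight swing and twist angle limits that restrict each joint to slight movement.
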PhysicsAsset /BRCosmetics/Characters/Player/Female/Medium/Bodies/F_MED_HighTower_Mango_Dark/Meshes/F_MED_HighTower_Mango_Dark_Hoop
bodies {
 dyn_hoop_01 [{shapeType=Capsule, length=1, radius=1}]
 dyn_hoop_02 [{shapeType=Capsule, length=1, radius=1, center=<1 0 0>, rotation=<rot -71.3921 0 0>}]
 dyn_parent_hoop_01 [{shapeType=Capsule, length=1, radius=1}]
}
constraints [
 {bone1Name=dyn_hoop_01, bone2Name=dyn_parent_hoop_01, swing1LimitDegrees=5, swing2LimitDegrees=15, twistLimitDegrees=3}
 {bone1Name=dyn_hoop_02, bone2Name=dyn_hoop_01, swing1LimitDegrees=5, swing2LimitDegrees=15, twistLimitDegrees=5}
]
