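# Physics asset for the hair mesh of the F_MED_HighTower_Mango_Dark character, providing collision and joint limits for dynamic hair.
# Seven hair chains (back, sides, front) of mostly box bodies hang from the head through swing-limited, tight-twist constraints.
# Capsules on head, neck, upper spine, clavicles and right upper arm form the body shapes, joined by free-rotation constraints.
# Free-rotation constraints also cross-link the right-side, back-right and back chains.
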PhysicsAsset /BRCosmetics/Characters/Player/Female/Medium/Bodies/F_MED_HighTower_Mango_Dark/Meshes/Parts/F_MED_HighTower_Mango_Dark_Hair
bodies {
 clavicle_l [{shapeType=Capsule, length=4.6352, radius=6.8578, center=<6.7335 -1.1727 -2.3232>, rotation=<rot 90 -1.1935 -6.1935>}]
 clavicle_r [{shapeType=Capsule, length=10.5687, radius=5.3578, center=<-5.204 2.1137 1.3915>, rotation=<rot 71.985 5.0002 0>}]
 dyn_hair_bk_01 [{shapeType=Box, extentX=6.4375, extentY=8.125, extentZ=1.0625, center=<3.2204 0 0>, rotation=<rot 4.9288 0 0>}]
 dyn_hair_bk_02 [{shapeType=Box, extentX=6.4375, extentY=8.125, extentZ=3.4337, center=<3.2204 0.0001 -0.6203>}]
 dyn_hair_bk_03 [{shapeType=Box, extentX=6.4375, extentY=10.625, extentZ=2.649, center=<3.2204 0 0.4306>}]
 dyn_hair_bk_04 [{shapeType=Box, extentX=6.4375, extentY=8.125, extentZ=3.1861, center=<4.0062 0 1.0365>, rotation=<rot -13.1142 0 0>}]
 dyn_hair_bk_l_01 [{shapeType=Box, extentX=6.4375, extentY=8.125, extentZ=1.25, center=<3.2204 0 0>}]
 dyn_hair_bk_l_02 [{shapeType=Box, extentX=7.4375, extentY=8.625, extentZ=5.5625, center=<3.2204 0 0>}]
 dyn_hair_bk_l_03 [{shapeType=Box, extentX=10.6305, extentY=13.693, extentZ=3.6875, center=<3.7924 1.8189 2.9995>, rotation=<rot 29.9998 18.6944 -15.1929>}]
 dyn_hair_bk_r_01 [{shapeType=Box, extentX=4.5933, extentY=8.125, extentZ=1.25, center=<1.901 0.0001 0.0004>}]
 dyn_hair_bk_r_02 [{shapeType=Box, extentX=15.3824, extentY=13.3606, extentZ=1.5924, center=<6.9971 0.0008 1.9168>, rotation=<rot 14.8074 0 23.544>}]
 dyn_hair_bk_r_03 [{shapeType=Box, extentX=8.6919, extentY=19.0667, extentZ=2.1875, center=<3.8116 0.0468 1.5147>, rotation=<rot 28.618 3.7233 19.8546>}]
 dyn_hair_fr_l_01 [{shapeType=Box, extentX=6.4375, extentY=8.125, extentZ=1.0625, center=<3.2204 0 0>}]
 dyn_hair_fr_l_02 [{shapeType=Box, extentX=6.4375, extentY=8.125, extentZ=1.375, center=<3.2204 0 0>}]
 dyn_hair_fr_l_03 [
  {shapeType=Capsule, length=2.9389, radius=1.0611, center=<4.29 3.5484 0.3866>, rotation=<rot 8.8304 -45.3645 -81.3792>}
  {shapeType=Capsule, length=1.9014, radius=1.361, center=<1.8673 1.6276 0.3525>, rotation=<rot 90 0 119.9984>}
 ]
 dyn_hair_fr_r_01 [{shapeType=Box, extentX=6.4375, extentY=8.125, extentZ=1.0625, center=<3.2204 0 0>}]
 dyn_hair_fr_r_02 [{shapeType=Box, extentX=7.9375, extentY=8.125, extentZ=1.375, center=<3.2204 0 0>}]
 dyn_hair_fr_r_03 [
  {shapeType=Capsule, length=2, radius=2.7812, center=<1.9985 -1.0461 1.2522>, rotation=<rot 0.015 90.6494 -77.6808>}
  {shapeType=Capsule, length=3.2847, radius=1.5354, center=<1.3963 -1.1406 0>, rotation=<rot -0.0004 24.9995 -89.9992>}
 ]
 dyn_hair_fr_r_04 [
  {shapeType=Capsule, length=4.75, radius=3.2812, center=<1.8774 -0.6012 0.0692>, rotation=<rot -0.1766 91.5259 -77.6803>}
  {shapeType=Capsule, length=2.5, radius=2.7812, center=<1.9444 0.4706 1.178>, rotation=<rot -28.6257 103.5598 -24.1096>}
 ]
 dyn_hair_l_01 [{shapeType=Box, extentX=6.4375, extentY=8.125, extentZ=3.375, center=<3.2204 0 0>}]
 dyn_hair_l_02 [{shapeType=Box, extentX=7.4375, extentY=8.125, extentZ=5.3125, center=<3.2204 0 0>}]
 dyn_hair_l_03 [{shapeType=Capsule, length=2.319, radius=1.7747, center=<1.5332 0 0>, rotation=<rot 0 0 -89.9992>}]
 dyn_hair_r_01 [{shapeType=Box, extentX=3.1158, extentY=8.125, extentZ=1.875, center=<1.593 0 0>}]
 dyn_hair_r_02 [{shapeType=Box, extentX=6.289, extentY=10.8964, extentZ=2.3125, center=<3.5962 0.0002 0>}]
 dyn_hair_r_03 [{shapeType=Box, extentX=11.4571, extentY=17.0788, extentZ=1.1438, center=<3.6437 0.0338 4.0519>, rotation=<rot 47.5699 1.8631 18.81>}]
 head [{shapeType=Capsule, length=5.5, radius=6.4062, center=<0.0001 -0.4045 0>, rotation=<rot 90 0 0>}]
 neck_01 [{shapeType=Capsule, length=4.5312, radius=6.125, center=<3.1352 0.7064 -0.8>, rotation=<rot 90 -18.9704 -16.4752>}]
 spine_05 [
  {shapeType=Capsule, length=15.3721, radius=9.1155, center=<-0.7291 -1.312 0>}
  {
   shapeType=Capsule
   length=5.75
   radius=3.25
   shapeName=spine_05_capsule
   center=<11.8634 4.2243 3.6316>
   rotation=<rot -3.2256 8.5732 90.4752>
  }
 ]
 upperarm_r [{shapeType=Capsule, length=14, radius=5.505, center=<-12.3869 5.0448 -3.6812>, rotation=<rot 0 -89.9997 91.2179>}]
}
constraints [
 {bone1Name=dyn_hair_bk_02, bone2Name=dyn_hair_bk_01, swing1LimitDegrees=15, swing2LimitDegrees=60, twistLimitDegrees=5}
 {bone1Name=dyn_hair_bk_03, bone2Name=dyn_hair_bk_02, swing1LimitDegrees=15, swing2LimitDegrees=60, twistLimitDegrees=5}
 {bone1Name=dyn_hair_bk_04, bone2Name=dyn_hair_bk_03, swing1LimitDegrees=15, swing2LimitDegrees=60, twistLimitDegrees=5}
 {bone1Name=dyn_hair_r_02, bone2Name=dyn_hair_r_01, swing1LimitDegrees=60, swing2LimitDegrees=40, twistLimitDegrees=5}
 {bone1Name=dyn_hair_r_03, bone2Name=dyn_hair_r_02, swing1LimitDegrees=45, swing2LimitDegrees=45, twistLimitDegrees=5}
 {bone1Name=dyn_hair_bk_r_02, bone2Name=dyn_hair_bk_r_01, swing1LimitDegrees=20, swing2LimitDegrees=60, twistLimitDegrees=5}
 {bone1Name=dyn_hair_bk_r_03, bone2Name=dyn_hair_bk_r_02, swing1LimitDegrees=20, swing2LimitDegrees=30, twistLimitDegrees=5}
 {bone1Name=dyn_hair_fr_r_02, bone2Name=dyn_hair_fr_r_01, swing1LimitDegrees=80, swing2LimitDegrees=38.8571, twistLimitDegrees=10}
 {bone1Name=dyn_hair_fr_r_03, bone2Name=dyn_hair_fr_r_02, swing1LimitDegrees=80, swing2LimitDegrees=40.5714, twistLimitDegrees=5}
 {bone1Name=dyn_hair_bk_l_02, bone2Name=dyn_hair_bk_l_01, swing1LimitDegrees=10, swing2LimitDegrees=30, twistLimitDegrees=5}
 {bone1Name=dyn_hair_bk_l_03, bone2Name=dyn_hair_bk_l_02, swing1LimitDegrees=10, swing2LimitDegrees=30, twistLimitDegrees=5}
 {bone1Name=dyn_hair_l_02, bone2Name=dyn_hair_l_01, swing1LimitDegrees=45, swing2LimitDegrees=45, twistLimitDegrees=5}
 {bone1Name=dyn_hair_l_03, bone2Name=dyn_hair_l_02, swing1LimitDegrees=45, swing2LimitDegrees=45, twistLimitDegrees=5}
 {bone1Name=dyn_hair_fr_l_02, bone2Name=dyn_hair_fr_l_01, swing1LimitDegrees=80, swing2LimitDegrees=80, twistLimitDegrees=10}
 {bone1Name=dyn_hair_fr_l_03, bone2Name=dyn_hair_fr_l_02, swing1LimitDegrees=80, swing2LimitDegrees=80, twistLimitDegrees=5}
 {bone1Name=dyn_hair_bk_01, bone2Name=head, swing1LimitDegrees=15, swing2LimitDegrees=60, twistLimitDegrees=5}
 {bone1Name=dyn_hair_r_01, bone2Name=head, swing1LimitDegrees=45, swing2LimitDegrees=45, twistLimitDegrees=5}
 {bone1Name=dyn_hair_bk_r_01, bone2Name=head, swing1LimitDegrees=20, swing2LimitDegrees=60, twistLimitDegrees=5}
 {bone1Name=dyn_hair_fr_r_01, bone2Name=head, swing1LimitDegrees=46.7143, swing2LimitDegrees=21, twistLimitDegrees=5}
 {bone1Name=dyn_hair_bk_l_01, bone2Name=head, swing1LimitDegrees=10, swing2LimitDegrees=60, twistLimitDegrees=5}
 {bone1Name=dyn_hair_l_01, bone2Name=head, swing1LimitDegrees=45, swing2LimitDegrees=45, twistLimitDegrees=5}
 {bone1Name=dyn_hair_fr_l_01, bone2Name=head, swing1LimitDegrees=45, swing2LimitDegrees=45, twistLimitDegrees=5}
 {
  bone1Name=head
  bone2Name=neck_01
  swing1LimitDegrees=45
  swing1Motion=Free
  swing2LimitDegrees=45
  swing2Motion=Free
  twistLimitDegrees=45
  twistMotion=Free
 }
 {
  bone1Name=clavicle_l
  bone2Name=spine_05
  swing1LimitDegrees=45
  swing1Motion=Free
  swing2LimitDegrees=45
  swing2Motion=Free
  twistLimitDegrees=45
  twistMotion=Free
 }
 {
  bone1Name=clavicle_r
  bone2Name=spine_05
  swing1LimitDegrees=45
  swing1Motion=Free
  swing2LimitDegrees=45
  swing2Motion=Free
  twistLimitDegrees=45
  twistMotion=Free
 }
 {
  bone1Name=neck_01
  bone2Name=spine_05
  swing1LimitDegrees=45
  swing1Motion=Free
  swing2LimitDegrees=45
  swing2Motion=Free
  twistLimitDegrees=45
  twistMotion=Free
 }
 {
  bone1Name=dyn_hair_r_03
  bone2Name=dyn_hair_bk_r_03
  swing1LimitDegrees=45
  swing1Motion=Free
  swing2LimitDegrees=45
  swing2Motion=Free
  twistLimitDegrees=45
  twistMotion=Free
 }
 {
  bone1Name=dyn_hair_bk_r_03
  bone2Name=dyn_hair_bk_03
  swing1LimitDegrees=45
  swing1Motion=Free
  swing2LimitDegrees=45
  swing2Motion=Free
  twistLimitDegrees=45
  twistMotion=Free
 }
 {bone1Name=dyn_hair_fr_r_04, bone2Name=dyn_hair_fr_r_03, swing1LimitDegrees=80, swing2LimitDegrees=40.5714, twistLimitDegrees=5}
 {
  bone1Name=upperarm_r
  bone2Name=clavicle_r
  swing1LimitDegrees=45
  swing1Motion=Free
  swing2LimitDegrees=45
  swing2Motion=Free
  twistLimitDegrees=45
  twistMotion=Free
 }
]
