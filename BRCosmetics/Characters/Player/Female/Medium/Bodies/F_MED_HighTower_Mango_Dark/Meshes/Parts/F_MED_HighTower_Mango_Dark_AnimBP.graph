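# Animation Blueprint for the F_MED_HighTower_Mango_Dark female body part, which follows the attached parent mesh and adds hair physics.
# The copied parent pose is cached, and bRigidBodyEnabled selects between it with an additive skeleton pose applied (true) or unmodified (false).
# The result feeds a RigidBody node using the Hair physics asset, simulated relative to neck_01 with scaled component motion and overridden gravity.
# The update event has no logic.
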AnimBlueprint /BRCosmetics/Characters/Player/Female/Medium/Bodies/F_MED_HighTower_Mango_Dark/Meshes/Parts/F_MED_HighTower_Mango_Dark_AnimBP
target_skeleton=&/F_MED_HighTower_Mango_Dark_Skeleton
graphs [
 AnimGraph {nodes [
   n1 = "Sequence Player (additive) 'F_MED_HighTower_Mango_Dark_Skeleton_Pose'"() {sequence=&/F_MED_HighTower_Mango_Dark_Skeleton_Pose}
   n2 = Apply Additive(Base=Use cached pose 'AnimCache'(), Additive=@n1) {alphaBoolBlend {blendOption=Linear}}
   n3 = Use cached pose 'AnimCache'()
   n4 = Get bRigidBodyEnabled()
   n5 = RigidBody(ComponentPose=Local To Component(LocalPose=Blend Poses by bool(BlendPose_0=@n2, BlendPose_1=@n3, bActiveValue=@n4) {blendType=Linear})) {
    bOverrideWorldGravity=true
    overridePhysicsAsset=&/F_MED_HighTower_Mango_Dark_Hair
    simulationSpace=BaseBoneSpace
    alphaBoolBlend {blendOption=Linear}
    baseBoneRef {boneName=neck_01}
    componentAppliedLinearAccClamp {y=1200}
    componentLinearAccScale=<0.3 0.35 0.3>
    componentLinearVelScale=<0.3 0.2 0.3>
    overrideWorldGravity {z=-980}
   }
   Output Pose(Result=Component To Local(ComponentPose=@n5)) {layerGroup=None}
   Save cached pose 'AnimCache'(Pose=Copy Pose From Mesh() {bUseAttachedParent=true}) {cacheName=AnimCache}
  ]}
 EventGraph {nodes [Event BlueprintUpdateAnimation()]}
]
variables []
